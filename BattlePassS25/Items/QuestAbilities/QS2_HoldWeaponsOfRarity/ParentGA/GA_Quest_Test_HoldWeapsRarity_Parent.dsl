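; Parent passive quest ability that grants credit when the player holds enough weapons of required rarities.
; On activation it checks the inventory and listens for pickup and filtered NPC service events, using a short delay so new items land first.
; It gathers items by rarity, then keeps unique ones matching required and excluded weapon tags.
; Reaching the required count broadcasts a custom verb for quest credit and ends the ability.
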(event Ability|EventFailedToActivatePassiveAbility
  (params (ActorInfo :type Gameplay%20Ability%20Actor%20Info%20Structure))
  (|CallParentFunction ActorInfo))

(event Ability|EventActivateAbility
  ;[0 0 0] Activate - Do an initial inventory check, then wait for the player to collect, purchase, or upgrade a weapon
  (|CallParentFunction)
  (Ability|Tasks|WaitGameplayEvent "(TagName=\"Event.CollectedPickup\")"
    (:then
      (bind (_asynctaskproxy _payload) (Ability|Tasks|WaitGameplayEvent "(TagName=\"Event.NPCService\")" 0 false false)
        (:then
          (Utilities|FlowControl|DoOnce false
            (:Completed
              _)))
        (:EventReceived
          (bind (_payload_eventtag _payload_instigator _payload_target _payload_optionalobject _payload_optionalobject2 _payload_contexthandle _payload_instigatortags) (Utilities|Struct|BreakGameplayEventData _payload))
          (if (GameplayTags|HasAnyTags _payload_instigatortags (Variables|RuntimeVariablesNoEditinChildren|GetValidNPCServices) true)
            _)))
      (Utilities|Array|Clear (Variables|RuntimeVariablesNoEditinChildren|GetPlayerWeaponsOfReqRarities))
      (Utilities|Array|Clear (Variables|RuntimeVariablesNoEditinChildren|GetPlayerItemsOfReqRarities))
      ;;[1 1 1] Delay a moment so the latest pickup item gets into the players inventory
      (Utilities|FlowControl|Delay 1.0)
      (Variables|RuntimeVariablesNoEditinChildren|SetOwningPlayerInventoryInterface (Fortnite|GetFortPlayerControllerfromActor (Variables|GAVars|Data|GetOwningPlayerPawn)))
      (CallFunction|CheckInventoryforRarity)
      (CallFunction|CheckHasRequiredWeaponsRarity))
    (:EventReceived
      (Development|PrintString "Item picked up" true true "(R=1.000000,G=0.000000,B=0.574233,A=1.000000)" 4.0))))

(event Custom|Check%20Inventory%20For%20Rarity
  ;[0 0 0] Get all items of the specified rarities and combine to one array
  (for _array_element (Variables|GAConditions|GetRequiredRarityTypes)
    (bind _items (InventoryManagement|GetItemsfromInventoryOwnerforRarity (Variables|RuntimeVariablesNoEditinChildren|GetOwningPlayerInventoryInterface) _array_element))
    (if (Utilities|Array|IsNotEmpty _items)
      (Utilities|Array|AppendArray (Variables|RuntimeVariablesNoEditinChildren|GetPlayerItemsOfReqRarities) _items))))

(event Custom|Check%20Has%20Required%20Weapons%20Rarity
  ;[0 0 0] Using the array of req rarities, check if they are weapons that count for this quest. If they do, add them to a new array weapons
  (bind _array_element (Utilities|Array|ForEachLoopwithBreak (Variables|RuntimeVariablesNoEditinChildren|GetPlayerItemsOfReqRarities))
    (:LoopBody
      (if (and (GameplayTags|HasAnyTags (CallFunction|GetOwnedGameplayTags _array_element) (Variables|_GAConditions|GetRequired%20Weapon%20Tags)) (not (GameplayTags|HasAnyTags (CallFunction|GetOwnedGameplayTags _array_element) (Variables|_GAConditions|GetExcludedWeaponTags) true)))
        (Utilities|Array|AddUnique (Variables|RuntimeVariablesNoEditinChildren|GetPlayerWeaponsOfReqRarities) _array_element)
        ;;[0.1198 0.1198 0.1198] Do we have the minimum count of weapons in this new array? If yes, set bool to True and break the loop
        (if (>= (Utilities|Array|Length (Variables|RuntimeVariablesNoEditinChildren|GetPlayerWeaponsOfReqRarities)) (Variables|_GAConditions|GetAmountToHold))
          (Variables|RuntimeVariablesNoEditinChildren|SetAllWeaponsFound true))
        (Variables|RuntimeVariablesNoEditinChildren|SetAllWeaponsFound false)))
    (:Completed
      _))
  ;;[0.1615 0.1615 0.1615] If bool is true at the end of the loop, send quest credit and end the ability
  (if (Variables|RuntimeVariablesNoEditinChildren|GetAllWeaponsFound)
    (CustomVerb|BroadcastCustomVerbMessage (CustomVerb|MakeCustomVerbMessage (Variables|GAVars|Data|GetOwningPlayerPawn) (Variables|_GAConditions|GetCustomVerbName)))
    (Ability|EndAbility)))

(event Ability|EventOnEndAbility
  (params (bWasCancelled :type Boolean)))

(fn GetOwnedGameplayTags
  (params (AssetInterface :type Gameplay%20Tag%20Asset%20Interface%20Interface))
  (return (GameplayTags|GetOwnedGameplayTags AssetInterface :!class /Script/GameplayTags.GameplayTagAssetInterface)))
